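# Physics asset for the long-skin TalonPane glider's skeletal mesh, covering its body, handles, wings and parachute.
# Every body is a single capsule on a named bone.
# Constraints form a tree from Body_main through each wing to its two flaps, and through Parachute_main to four corners, each continuing to a base rope.
# All constraints share uniform swing and twist limit angles, while the two handle bases have no constraints.
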PhysicsAsset /BRCosmetics/Gadgets/Assets/VinderTech_GliderChute/Glider_TalonPane/Skins/Long/Meshes/TalonPane_Long_Glider_PhysicsAsset
bodies {
 Body_main [{shapeType=Capsule, length=68.4447, radius=43.6252, center=<0.0018 14.4698 6.2713>, rotation=<rot 0 90.0041 -89.9981>}]
 Handle_Base_l [{shapeType=Capsule, length=15.3149, radius=6.6734, center=<-0.4424 8.2574 -8.3585>, rotation=<rot 0 2.4432 -66.7605>}]
 Handle_Base_r [{shapeType=Capsule, length=15.3149, radius=6.6734, center=<0.356 -8.1071 8.3657>, rotation=<rot 0 2.4432 -66.7605>}]
 Parachute_l_backcorner [{shapeType=Capsule, length=49.9419, radius=32.2778, center=<5.8271 -20.0193 -5.5902>, rotation=<rot 0 31.3699 -56.7761>}]
 Parachute_l_basebackrope [{shapeType=Capsule, length=28.032, radius=11.8313, center=<10.9982 -6.7962 -4.5975>, rotation=<rot 0 -106.6706 -81.388>}]
 Parachute_l_basefrontrope [{shapeType=Capsule, length=25.1154, radius=11.1712, center=<7.6756 -3.095 5.9794>, rotation=<rot 0 86.1253 -77.5626>}]
 Parachute_l_frontcorner [{shapeType=Capsule, length=48.9999, radius=32.261, center=<-7.9098 -19.201 -6.1199>, rotation=<rot 0 -45.9626 -58.5773>}]
 Parachute_main [{shapeType=Capsule, length=64.1409, radius=14.5627, center=<0 -3.3636 -117.1822>, rotation=<rot 0 -90 -90>}]
 Parachute_r_backtcorner [{shapeType=Capsule, length=49.8182, radius=36.88, center=<1.9162 20.8242 5.1562>, rotation=<rot 0 50.3965 -69.804>}]
 Parachute_r_basebackrope [{shapeType=Capsule, length=28.032, radius=12.1049, center=<-8.6324 7.1248 -8.3356>, rotation=<rot 0 72.0072 -63.5677>}]
 Parachute_r_basefrontrope [{shapeType=Capsule, length=25.1154, radius=11.1712, center=<-7.6756 3.095 -5.9794>, rotation=<rot 0 86.1252 -77.5626>}]
 Parachute_r_frontcorner [{shapeType=Capsule, length=48.3239, radius=33.5017, center=<5.6576 19.7219 6.9318>, rotation=<rot 0 -53.5649 -65.4378>}]
 Wing_l [{shapeType=Capsule, length=95.1749, radius=57.6341, center=<-52.3044 8.0609 -1.1345>, rotation=<rot 0 -83.824 -86.4102>}]
 Wing_l_flapa [{shapeType=Capsule, length=35.2888, radius=9.5047, center=<1.2347 6.9411 0.5065>, rotation=<rot 0 -90.654 -88.9571>}]
 Wing_l_flapc [{shapeType=Capsule, length=51.8958, radius=17.3463, center=<5.6428 11.7039 -0.1058>, rotation=<rot 0 -81.4727 -89.2959>}]
 Wing_r [{shapeType=Capsule, length=95.0954, radius=57.6215, center=<52.3841 -8.0468 1.1274>, rotation=<rot 0 -83.8358 -86.4122>}]
 Wing_r_flapa [{shapeType=Capsule, length=35.2888, radius=9.5047, center=<-1.0721 -6.9991 -0.5239>, rotation=<rot 0 -90.654 -88.9571>}]
 Wing_r_flapc [{shapeType=Capsule, length=51.8958, radius=17.3463, center=<-5.4864 -11.7771 0.0887>, rotation=<rot 0 -81.4727 -89.2959>}]
}
constraints [
 {bone1Name=Wing_l, bone2Name=Body_main, swing1LimitDegrees=45, swing2LimitDegrees=45, twistLimitDegrees=45}
 {bone1Name=Wing_l_flapc, bone2Name=Wing_l, swing1LimitDegrees=45, swing2LimitDegrees=45, twistLimitDegrees=45}
 {bone1Name=Wing_l_flapa, bone2Name=Wing_l, swing1LimitDegrees=45, swing2LimitDegrees=45, twistLimitDegrees=45}
 {bone1Name=Wing_r, bone2Name=Body_main, swing1LimitDegrees=45, swing2LimitDegrees=45, twistLimitDegrees=45}
 {bone1Name=Wing_r_flapc, bone2Name=Wing_r, swing1LimitDegrees=45, swing2LimitDegrees=45, twistLimitDegrees=45}
 {bone1Name=Wing_r_flapa, bone2Name=Wing_r, swing1LimitDegrees=45, swing2LimitDegrees=45, twistLimitDegrees=45}
 {bone1Name=Parachute_main, bone2Name=Body_main, swing1LimitDegrees=45, swing2LimitDegrees=45, twistLimitDegrees=45}
 {bone1Name=Parachute_r_backtcorner, bone2Name=Parachute_main, swing1LimitDegrees=45, swing2LimitDegrees=45, twistLimitDegrees=45}
 {bone1Name=Parachute_r_basebackrope, bone2Name=Parachute_r_backtcorner, swing1LimitDegrees=45, swing2LimitDegrees=45, twistLimitDegrees=45}
 {bone1Name=Parachute_r_frontcorner, bone2Name=Parachute_main, swing1LimitDegrees=45, swing2LimitDegrees=45, twistLimitDegrees=45}
 {bone1Name=Parachute_r_basefrontrope, bone2Name=Parachute_r_frontcorner, swing1LimitDegrees=45, swing2LimitDegrees=45, twistLimitDegrees=45}
 {bone1Name=Parachute_l_frontcorner, bone2Name=Parachute_main, swing1LimitDegrees=45, swing2LimitDegrees=45, twistLimitDegrees=45}
 {bone1Name=Parachute_l_basefrontrope, bone2Name=Parachute_l_frontcorner, swing1LimitDegrees=45, swing2LimitDegrees=45, twistLimitDegrees=45}
 {bone1Name=Parachute_l_backcorner, bone2Name=Parachute_main, swing1LimitDegrees=45, swing2LimitDegrees=45, twistLimitDegrees=45}
 {bone1Name=Parachute_l_basebackrope, bone2Name=Parachute_l_backcorner, swing1LimitDegrees=45, swing2LimitDegrees=45, twistLimitDegrees=45}
]
